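# Additive UI material for a glowing green aura in the ToyFox boss bar HUD.
# Radial gradients form a parameterized ring, a two-layer center glow and masks, alongside a banded swirl-texture pattern and a three-tone green blend.
# Sparkles use polar-mapped UVs panned by the Starbust Location parameter, sampled with explicit derivatives.
# A time-driven cosine gradient is also declared; reroute usages are unnamed, so the final combination is unclear.
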Material /ToyFoxGameplay/UI/Widgets/HUD/Assets/BossBar/M_UI_ToyFox_Aura
graph {
 properties {bAutomaticallySetUsageInEditor=false, blendMode=BLEND_Additive, materialDomain=MD_UI, usedShadingModels=Default Lit}
 nodes [
  Constant() {r=0.224}
  Constant() {r=1}
  Constant() {r=0.072}
  Constant() {r=0.336}
  Frac(None=Time())
  LinearInterpolate(
   A=LinearInterpolate(A=Constant3Vector() {constant=<rgba 0.1827 0.776 0 1>}, B=Constant3Vector() {constant=<rgba 0.6814 1 0 1>}, Alpha=NamedRerouteUsage())
   B=Constant3Vector() {constant=<rgba 1.1296 2 0 1>}
   Alpha=NamedRerouteUsage()
  )
  n1 = TextureSample() {texture=/Game/Global/Textures/Default/Flowmaps/T_SwirlFlow_02_D}
  n2 = Multiply(A=Multiply(A=SmoothStep(Value=@n1.R) {constMax=0.3, constMin=0.5}, B=SmoothStep(Value=@n1.R) {constMax=0.5, constMin=0.3})) {constB=4}
  n3 = NamedRerouteUsage()
  Multiply(A=Add(A=Add(A=@n2, B=@n3), B=NamedRerouteUsage()), B=NamedRerouteUsage())
  Multiply(A=NamedRerouteUsage(), B=NamedRerouteUsage())
  n4 = TextureCoordinate()
  n5 = Multiply(A=ScalarParameter() {defaultValue=0.9593, parameterName=RingSize}) {constB=0.2}
  n6 = Constant() {r=24}
  n7 = MaterialFunctionCall(UVs=@n4, Radius=@n5, Density=@n6) {materialFunction=&RadialGradientExponential}
  n8 = OneMinus(None=MaterialFunctionCall(UVs=@n4, Radius=Subtract(A=@n5) {constB=0.03}, Density=@n6) {materialFunction=&RadialGradientExponential}.RadialGradientExponential)
  NamedRerouteDeclaration(None=Multiply(A=DotProduct(A=@n7.RadialGradientExponential, B=@n8), B=ScalarParameter() {parameterName=RingAlpha})) {name=Ringo, nodeColor=<rgba 0.4297 1 0 1>}
  n9 = TextureCoordinate()
  n10 = ScalarParameter() {defaultValue=0.112, parameterName=CenterGlowSize}
  n11 = Multiply(
   A=MaterialFunctionCall(UVs=@n9, Radius=@n10, Density=Constant() {r=2.5815}) {materialFunction=&RadialGradientExponential}.RadialGradientExponential
   B=ScalarParameter() {defaultValue=0.78, parameterName=CenterGlowAlpha}
  )
  n12 = Multiply(
   A=MaterialFunctionCall(UVs=@n9, Radius=Add(A=@n10) {constB=0.112}, Density=Constant() {r=1.873}) {materialFunction=&RadialGradientExponential}.RadialGradientExponential
   B=Constant() {r=0.312}
  )
  NamedRerouteDeclaration(None=Add(A=@n11, B=@n12)) {name=CenterGlow, nodeColor=<rgba 0.4297 1 0 1>}
  n13 = Constant() {r=1.1493}
  n14 = Constant() {r=1.3692}
  n15 = MaterialFunctionCall(
   UVs=MaterialFunctionCall(UVs=TextureCoordinate(), Texture Scale=AppendVector(A=@n13, B=@n14)) {materialFunction=&ScaleUVsByCenter}.UVs
   Radius=Constant() {r=0.378}
   Density=Constant() {r=1.1}
  ) {materialFunction=&RadialGradientExponential}
  NamedRerouteDeclaration(None=@n15.RadialGradientExponential) {name=RadialMask, nodeColor=<rgba 0.4297 1 0 1>}
  NamedRerouteDeclaration(None=@n11) {name=InnerGlow, nodeColor=<rgba 0 0 0 1>}
  NamedRerouteDeclaration(None=@n12) {name=OuterGlow, nodeColor=<rgba 0 0 0 1>}
  n16 = MaterialFunctionCall() {materialFunction=&MF_UI_UVScaling}
  n17 = MaterialFunctionCall(Vector or UVs=Subtract(A=@n16.Scaled to Fit) {constB=0.5}) {materialFunction=&VectorToRadialValue}
  n18 = Multiply(A=@n17.Linear Distance, B=Constant() {r=-0.1})
  n19 = ComponentMask(None=@n17.Radial Coordinates) {r=true}
  n20 = Constant() {r=1}
  n21 = Panner(
   Coordinate=AppendVector(A=ComponentMask(None=@n18) {r=true}, B=Multiply(A=@n19, B=@n20))
   Time=LinearInterpolate(Alpha=ScalarParameter() {defaultValue=0.324, parameterName=Starbust Location}) {constB=5}
   Speed=AppendVector(A=Constant() {r=0.2}, B=Constant())
  ) {speedX=1}
  n22 = Multiply(A=@n16.Scaled to Fit)
  n23 = DDX(Value=@n22)
  n24 = DDY(Value=@n22)
  NamedRerouteDeclaration(None=Power(Base=TextureSample!(UVs=@n21, Tex, "DDX(UVs)"=@n23, "DDY(UVs)"=@n24, Apply View MipBias) {mipValueMode=TMVM_Derivative, texture=/Game/UI/HUD/Textures/UIPatterns/T_UI_SparklesSingle}.R)) {name=Sparkles, nodeColor=<rgba 0 0 0 1>}
  n25 = Multiply(A=Time()) {constB=5}
  n26 = Constant() {r=0.3}
  n27 = Constant()
  NamedRerouteDeclaration(None=MaterialFunctionCall!(
    Gradient
    a
    b
    c
    d=MaterialFunctionCall(R=@n25, G=@n26, B=@n27) {materialFunction=&AppendMany}.RGB
   ) {materialFunction=/FortUILibrary/MaterialFunctions/Color/MF_UI_CosineGradient}.Result) {nodeColor=<rgba 1 0.0938 0 1>}
  NamedRerouteDeclaration(None=Power(Base=OneMinus(None=@n15.RadialGradientExponential)) {constExponent=10}) {name=Inner Radial Mask, nodeColor=<rgba 0 0.5547 1 1>}
  NamedRerouteUsage()
 ]
}
